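# Temporary Level Sequence for LimeRock shot 100 that animates four possessed FortPlayerMannequin stand-ins, P0 to P3.
# Each mannequin layers a zero-based offset transform section over its world-space placement, with most moving and turning as the animated window opens.
# After a short first clip, each plays a chain of mostly equal-length, overlapping skeletal animation sections.
# Visibility keys hide and re-show the mannequins, and binding lifetime tracks limit them to that animated window.
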LevelSequence /LimeRock/Temp/Sequence/LimeRock_100_anm_Players_TEMP
bindings [
 FortPlayerMannequin_P3 {
  id=#1
  possessed_class=FortPlayerMannequin
  tracks [
   {
    type=MovieSceneSkeletalAnimationTrack
    sections [
     {type=MovieSceneSkeletalAnimationSection, end_frame=9049, start_frame=8905}
     {type=MovieSceneSkeletalAnimationSection, end_frame=9444, start_frame=9049}
     {type=MovieSceneSkeletalAnimationSection, end_frame=9752, start_frame=9357}
     {type=MovieSceneSkeletalAnimationSection, end_frame=10058, start_frame=9663}
     {type=MovieSceneSkeletalAnimationSection, end_frame=10366, start_frame=9971}
     {type=MovieSceneSkeletalAnimationSection, end_frame=10679, start_frame=10284}
     {type=MovieSceneSkeletalAnimationSection, end_frame=10987, start_frame=10592}
     {type=MovieSceneSkeletalAnimationSection, end_frame=11293, start_frame=10898}
     {type=MovieSceneSkeletalAnimationSection, end_frame=11601, start_frame=11206}
     {type=MovieSceneSkeletalAnimationSection, end_frame=11828, start_frame=11433}
    ]
   }
   {
    type=MovieScene3DTransformTrack
    sections [
     {
      type=MovieScene3DTransformSection
      channels [
       Location.X {type=MovieSceneScriptingDoubleChannel, default=-90309, keys [[1655,-90206.3984],[8905,-90209.7109],[9049,-90309]]}
       Location.Y {type=MovieSceneScriptingDoubleChannel, default=-95175, thinned_from=3, keys [[1655,-94827.7578],[9049,-95175]]}
       Location.Z {type=MovieSceneScriptingDoubleChannel, default=1082.5349, const [995.2323,3]}
       Rotation.X {type=MovieSceneScriptingDoubleChannel, default=0, const [0,3]}
       Rotation.Y {type=MovieSceneScriptingDoubleChannel, default=0, const [0,3]}
       Rotation.Z {type=MovieSceneScriptingDoubleChannel, default=104.6261, keys [[1655,104.6261],[8905,92.3664],[9049,-82.6011]]}
       Scale.X {type=MovieSceneScriptingDoubleChannel, default=1, const [1,3]}
       Scale.Y {type=MovieSceneScriptingDoubleChannel, default=1, const [1,3]}
       Scale.Z {type=MovieSceneScriptingDoubleChannel, default=1, const [1,3]}
      ]
     }
     {
      type=MovieScene3DTransformSection
      channels [
       Location.X {type=MovieSceneScriptingDoubleChannel, default=0, keys [[1655,0],[8905,-455]]}
       Location.Y {type=MovieSceneScriptingDoubleChannel, default=0, keys [[1655,0],[8905,-526]]}
       Location.Z {type=MovieSceneScriptingDoubleChannel, default=0, keys [[1655,0],[8905,34]]}
       Rotation.X {type=MovieSceneScriptingDoubleChannel, default=0, const [0,2]}
       Rotation.Y {type=MovieSceneScriptingDoubleChannel, default=0, const [0,2]}
       Rotation.Z {type=MovieSceneScriptingDoubleChannel, default=0, keys [[1655,0],[8905,11.3742]]}
       Scale.X {type=MovieSceneScriptingDoubleChannel, default=0, const [0,2]}
       Scale.Y {type=MovieSceneScriptingDoubleChannel, default=0, const [0,2]}
       Scale.Z {type=MovieSceneScriptingDoubleChannel, default=0, const [0,2]}
      ]
     }
    ]
   }
   {
    type=MovieSceneVisibilityTrack
    sections [{
      type=MovieSceneVisibilitySection
      channels [None {type=MovieSceneScriptingBoolChannel, default=false, keys [[1355,true],[3600,false],[8905,true]]}]
     }]
   }
   {
    type=MovieSceneBindingLifetimeTrack
    display_name=Binding Lifetime
    sections [{type=MovieSceneBindingLifetimeSection, end_frame=11822, start_frame=8905}]
   }
  ]
 }
 FortPlayerMannequin_P2 {
  id=#2
  possessed_class=FortPlayerMannequin
  tracks [
   {
    type=MovieScene3DTransformTrack
    sections [
     {
      type=MovieScene3DTransformSection
      channels [
       Location.X {type=MovieSceneScriptingDoubleChannel, default=-90186, keys [[1655,-90370.5312],[8905,-90186],[9049,-90583]]}
       Location.Y {type=MovieSceneScriptingDoubleChannel, default=-95472, keys [[1655,-95335.2109],[8905,-95472],[9049,-95292]]}
       Location.Z {type=MovieSceneScriptingDoubleChannel, default=1080.9336, thinned_from=3, keys [[1655,1007.5369],[9049,1007.5369]]}
       Rotation.X {type=MovieSceneScriptingDoubleChannel, default=0, const [0,3]}
       Rotation.Y {type=MovieSceneScriptingDoubleChannel, default=0, const [0,3]}
       Rotation.Z {type=MovieSceneScriptingDoubleChannel, default=70, keys [[1655,47.5011],[8905,70],[9049,208]]}
       Scale.X {type=MovieSceneScriptingDoubleChannel, default=1, const [1,3]}
       Scale.Y {type=MovieSceneScriptingDoubleChannel, default=1, const [1,3]}
       Scale.Z {type=MovieSceneScriptingDoubleChannel, default=1, const [1,3]}
      ]
     }
     {
      type=MovieScene3DTransformSection
      channels [
       Location.X {type=MovieSceneScriptingDoubleChannel, default=0, keys [[1655,0],[8905,-463],[9049,-311]]}
       Location.Y {type=MovieSceneScriptingDoubleChannel, default=0, keys [[1655,0],[8905,-406],[9049,-582]]}
       Location.Z {type=MovieSceneScriptingDoubleChannel, default=0, keys [[1655,0],[8905,21]]}
       Rotation.X {type=MovieSceneScriptingDoubleChannel, default=0, const [0,2]}
       Rotation.Y {type=MovieSceneScriptingDoubleChannel, default=0, const [0,2]}
       Rotation.Z {type=MovieSceneScriptingDoubleChannel, default=0, keys [[1655,0],[8905,18.8063]]}
       Scale.X {type=MovieSceneScriptingDoubleChannel, default=0, const [0,2]}
       Scale.Y {type=MovieSceneScriptingDoubleChannel, default=0, const [0,2]}
       Scale.Z {type=MovieSceneScriptingDoubleChannel, default=0, const [0,2]}
      ]
     }
    ]
   }
   {
    type=MovieSceneVisibilityTrack
    sections [{
      type=MovieSceneVisibilitySection
      channels [None {type=MovieSceneScriptingBoolChannel, default=false, keys [[1355,true],[3600,false],[8905,true]]}]
     }]
   }
   {
    type=MovieSceneSkeletalAnimationTrack
    sections [
     {type=MovieSceneSkeletalAnimationSection, end_frame=9049, start_frame=8905}
     {type=MovieSceneSkeletalAnimationSection, end_frame=9465, start_frame=9049}
     {type=MovieSceneSkeletalAnimationSection, end_frame=9822, start_frame=9406}
     {type=MovieSceneSkeletalAnimationSection, end_frame=10118, start_frame=9702}
     {type=MovieSceneSkeletalAnimationSection, end_frame=10456, start_frame=10040}
     {type=MovieSceneSkeletalAnimationSection, end_frame=10784, start_frame=10368}
     {type=MovieSceneSkeletalAnimationSection, end_frame=11154, start_frame=10738}
     {type=MovieSceneSkeletalAnimationSection, end_frame=11492, start_frame=11076}
     {type=MovieSceneSkeletalAnimationSection, end_frame=11820, start_frame=11404}
    ]
   }
   {
    type=MovieSceneBindingLifetimeTrack
    display_name=Binding Lifetime
    sections [{type=MovieSceneBindingLifetimeSection, end_frame=11822, start_frame=8905}]
   }
  ]
 }
 FortPlayerMannequin_P0 {
  id=#3
  possessed_class=FortPlayerMannequin
  tracks [
   {
    type=MovieSceneSkeletalAnimationTrack
    sections [
     {type=MovieSceneSkeletalAnimationSection, end_frame=9049, start_frame=8905}
     {type=MovieSceneSkeletalAnimationSection, end_frame=9526, start_frame=9049}
     {type=MovieSceneSkeletalAnimationSection, end_frame=9930, start_frame=9453}
     {type=MovieSceneSkeletalAnimationSection, end_frame=10352, start_frame=9875}
     {type=MovieSceneSkeletalAnimationSection, end_frame=10825, start_frame=10348}
     {type=MovieSceneSkeletalAnimationSection, end_frame=11359, start_frame=10724}
     {type=MovieSceneSkeletalAnimationSection, end_frame=11832, start_frame=11197}
    ]
   }
   {
    type=MovieScene3DTransformTrack
    sections [
     {
      type=MovieScene3DTransformSection
      channels [
       Location.X {type=MovieSceneScriptingDoubleChannel, default=-90367, thinned_from=3, keys [[1655,-90206.3984],[9049,-90362]]}
       Location.Y {type=MovieSceneScriptingDoubleChannel, default=-95306, thinned_from=3, keys [[1655,-94827.7578],[9049,-95228]]}
       Location.Z {type=MovieSceneScriptingDoubleChannel, default=1082.5349, thinned_from=3, keys [[1655,995.2323],[9049,995.2323]]}
       Rotation.X {type=MovieSceneScriptingDoubleChannel, default=0, const [0,3]}
       Rotation.Y {type=MovieSceneScriptingDoubleChannel, default=0, const [0,3]}
       Rotation.Z {type=MovieSceneScriptingDoubleChannel, default=104.6261, keys [[1655,104.6261],[8905,92.3664],[9049,-82.6011]]}
       Scale.X {type=MovieSceneScriptingDoubleChannel, default=1, const [1,3]}
       Scale.Y {type=MovieSceneScriptingDoubleChannel, default=1, const [1,3]}
       Scale.Z {type=MovieSceneScriptingDoubleChannel, default=1, const [1,3]}
      ]
     }
     {
      type=MovieScene3DTransformSection
      channels [
       Location.X {type=MovieSceneScriptingDoubleChannel, default=0, keys [[1655,0],[8905,-455]]}
       Location.Y {type=MovieSceneScriptingDoubleChannel, default=0, keys [[1655,0],[8905,-526]]}
       Location.Z {type=MovieSceneScriptingDoubleChannel, default=0, keys [[1655,0],[8905,34]]}
       Rotation.X {type=MovieSceneScriptingDoubleChannel, default=0, const [0,2]}
       Rotation.Y {type=MovieSceneScriptingDoubleChannel, default=0, const [0,2]}
       Rotation.Z {type=MovieSceneScriptingDoubleChannel, default=0, keys [[1655,0],[8905,11.3742]]}
       Scale.X {type=MovieSceneScriptingDoubleChannel, default=0, const [0,2]}
       Scale.Y {type=MovieSceneScriptingDoubleChannel, default=0, const [0,2]}
       Scale.Z {type=MovieSceneScriptingDoubleChannel, default=0, const [0,2]}
      ]
     }
    ]
   }
   {
    type=MovieSceneVisibilityTrack
    sections [{
      type=MovieSceneVisibilitySection
      channels [None {type=MovieSceneScriptingBoolChannel, default=false, keys [[1355,true],[3600,false],[8905,true]]}]
     }]
   }
   {
    type=MovieSceneBindingLifetimeTrack
    display_name=Binding Lifetime
    sections [{type=MovieSceneBindingLifetimeSection, end_frame=11822, start_frame=8905}]
   }
  ]
 }
 FortPlayerMannequin_P1 {
  id=#4
  possessed_class=FortPlayerMannequin
  tracks [
   {
    type=MovieScene3DTransformTrack
    sections [
     {
      type=MovieScene3DTransformSection
      channels [
       Location.X {type=MovieSceneScriptingDoubleChannel, default=-90373, thinned_from=3, keys [[1655,-90165.9141],[9049,-90321]]}
       Location.Y {type=MovieSceneScriptingDoubleChannel, default=-95407, keys [[1655,-95433.6172],[8905,-95407],[9049,-95278]]}
       Location.Z {type=MovieSceneScriptingDoubleChannel, default=1104.9554, const [1006.8921,3]}
       Rotation.X {type=MovieSceneScriptingDoubleChannel, default=0, const [0,3]}
       Rotation.Y {type=MovieSceneScriptingDoubleChannel, default=0, const [0,3]}
       Rotation.Z {type=MovieSceneScriptingDoubleChannel, default=91, keys [[1655,68.5154],[8905,91],[9049,-111.6268]]}
       Scale.X {type=MovieSceneScriptingDoubleChannel, default=1, const [1,3]}
       Scale.Y {type=MovieSceneScriptingDoubleChannel, default=1, const [1,3]}
       Scale.Z {type=MovieSceneScriptingDoubleChannel, default=1, const [1,3]}
      ]
     }
     {
      type=MovieScene3DTransformSection
      channels [
       Location.X {type=MovieSceneScriptingDoubleChannel, default=0, keys [[1655,0],[8905,-463],[9049,-423]]}
       Location.Y {type=MovieSceneScriptingDoubleChannel, default=0, keys [[1655,0],[8905,-542],[9049,-494]]}
       Location.Z {type=MovieSceneScriptingDoubleChannel, default=0, keys [[1655,0],[8905,23],[9049,23]]}
       Rotation.X {type=MovieSceneScriptingDoubleChannel, default=0, const [0,3]}
       Rotation.Y {type=MovieSceneScriptingDoubleChannel, default=0, const [0,3]}
       Rotation.Z {type=MovieSceneScriptingDoubleChannel, default=0, keys [[1655,0],[8905,-11.1191],[9049,0]]}
       Scale.X {type=MovieSceneScriptingDoubleChannel, default=0, const [0,3]}
       Scale.Y {type=MovieSceneScriptingDoubleChannel, default=0, const [0,3]}
       Scale.Z {type=MovieSceneScriptingDoubleChannel, default=0, const [0,3]}
      ]
     }
    ]
   }
   {
    type=MovieSceneVisibilityTrack
    sections [{
      type=MovieSceneVisibilitySection
      channels [None {type=MovieSceneScriptingBoolChannel, default=false, keys [[1355,true],[3600,false],[8905,true]]}]
     }]
   }
   {
    type=MovieSceneSkeletalAnimationTrack
    sections [
     {type=MovieSceneSkeletalAnimationSection, end_frame=9049, start_frame=8905}
     {type=MovieSceneSkeletalAnimationSection, end_frame=9559, start_frame=9049}
     {type=MovieSceneSkeletalAnimationSection, end_frame=9936, start_frame=9426}
     {type=MovieSceneSkeletalAnimationSection, end_frame=10335, start_frame=9825}
     {type=MovieSceneSkeletalAnimationSection, end_frame=10712, start_frame=10202}
     {type=MovieSceneSkeletalAnimationSection, end_frame=11119, start_frame=10609}
     {type=MovieSceneSkeletalAnimationSection, end_frame=11496, start_frame=10986}
     {type=MovieSceneSkeletalAnimationSection, end_frame=11844, start_frame=11334}
    ]
   }
   {
    type=MovieSceneBindingLifetimeTrack
    display_name=Binding Lifetime
    sections [{type=MovieSceneBindingLifetimeSection, end_frame=11822, start_frame=8905}]
   }
  ]
 }
]
display_rate {denominator=1, numerator=60}
tick_resolution {denominator=1, numerator=24000}
tracks []
